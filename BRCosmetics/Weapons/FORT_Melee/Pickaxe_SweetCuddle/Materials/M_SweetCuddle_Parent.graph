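# Parent material for the SweetCuddle pickaxe, built on the shared character PreFX, skin FX and PostFX functions for skeletal meshes.
# MF_Sweet, borrowed from the Sweet Teriyaki Red character folder, layers effects over CharacterSkinFX using the SkinFX mask.
# An optional static switch color-burns emissive and overrides specular and roughness in a selectable mask channel, blue by default.
# The burn amount follows time of day, with a separate scale that eases darkening at night.
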Material /BRCosmetics/Weapons/FORT_Melee/Pickaxe_SweetCuddle/Materials/M_SweetCuddle_Parent
graph {
 properties {
  bUseMaterialAttributes=true
  bUsedWithMeshDeformer=true
  bUsedWithMorphTargets=true
  bUsedWithSkeletalMesh=true
  pixelDepthOffsetMode=PDOM_Legacy
  refractionMethod=RM_IndexOfRefraction
  usedShadingModels=Default Lit
 }
 nodes [
  n1 = MaterialFunctionCall() {materialFunction=&BaseFN_CharacterFunction_PreFX}
  n2 = TextureObjectParameter() {group=Skin FX, parameterName=SkinFX_Mask}
  n3 = MaterialFunctionCall!(
   In=MaterialFunctionCall(
    PreskinnedPosition=@n1.Pre skinned Local Position
    Gmap=@n1.Pre skinned Local Position
    MaterialAttributes=@n1.Pre skinned Local Position
    AO=@n1.Pre skinned Local Position
   ) {materialFunction=&CharacterSkinFX}.Result
   FX Mask=@n2
   Pre Skinned Position=@n1.Pre skinned Local Position
  ) {materialFunction=/Game/Characters/Player/Male/Medium/Bodies/M_MED_Sweet_Teriyaki_Red/Materials/MF_Sweet}
  n4 = GetMaterialAttributes(None=@n3.Result) {attributeGetTypes [#1]}
  n5 = ScalarParameter() {defaultValue=0.8, group=SweetCuddle, parameterName=DarkeningAmt}
  n6 = OneMinus(None=Multiply(A=OneMinus(None=@n5), B=ScalarParameter() {defaultValue=0.4, group=SweetCuddle, parameterName=DarkeningAmt_NightScale}) {constB=0.5})
  n7 = SetMaterialAttributes!(
   MaterialAttributes=@n4.MaterialAttributes
   Emissive Color=MaterialFunctionCall!(
    Base=@n4.Emissive Color
    Blend=MaterialFunctionCall(Value_Morning=@n5, Value_Day=@n5, Value_Evening=@n5, Value_Night=@n6) {materialFunction=&Scalar_Time_of_Day}.Result
   ) {materialFunction=/Engine/Functions/Engine_MaterialFunctions03/Blends/Blend_ColorBurn}.Result
   Specular=ScalarParameter() {group=SweetCuddle, parameterName=DarkeningSpec}
   Roughness=ScalarParameter() {defaultValue=1, group=SweetCuddle, parameterName=DarkeningRoughness}
  ) {attributeSetTypes [#1, #2, #3]}
  n8 = StaticComponentMaskParameter(None=TextureSample(Tex=@n2).RGBA) {defaultB=true, group=SweetCuddle, parameterName=DarkeningMaskChannel}
  MaterialFunctionCall(
   AO=@n1.AO
   Distance Blend=@n1.AO
   Fresnel=@n1.AO
   Gmap=@n1.AO
   Pre Skinned Local Position=@n1.AO
   Material=StaticSwitchParameter(True=BlendMaterialAttributes(A=@n3.Result, B=@n7, Alpha=@n8), False=@n3.Result) {group=SweetCuddle, parameterName=Darken Emissive}
  ) {materialFunction=&BaseFN_CharacterFunction_PostFX}
 ]
}
