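# Physics asset for the SweetCuddle pickaxe, a two-handed melee cosmetic, giving its bear head and tufts jointed physics bodies.
# Capsule bodies cover the handle, the bear head and two tufts.
# The head is constrained to the handle with tight swing and twist limits, and both tufts hang off the head with looser limits.
# The dyn_ bone names suggest these parts are meant for secondary jiggle motion.
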PhysicsAsset /BRCosmetics/Weapons/FORT_Melee/Pickaxe_SweetCuddle/Meshes/SweetCuddle_Axe_Physics
bodies {
 Melee_TwoHanded_Handle [{shapeType=Capsule, length=50, radius=3}]
 dyn_bearHead [{shapeType=Capsule, length=2, radius=2, shapeName=dyn_bearHead_capsule, center=<5 0 0>, rotation=<rot 0 0 90>}]
 dyn_tuft_A [{shapeType=Capsule, length=3.8828, radius=4, shapeName=dyn_tuft_A_capsule, center=<4.0474 0 0>, rotation=<rot 0 -90 -90>}]
 dyn_tuft_B [{shapeType=Capsule, length=3.6946, radius=3, shapeName=dyn_tuft_B_capsule, center=<3.3224 0 0>, rotation=<rot 0 90 -90>}]
}
constraints [
 {bone1Name=dyn_bearHead, bone2Name=Melee_TwoHanded_Handle, swing1LimitDegrees=4, swing2LimitDegrees=4, twistLimitDegrees=4}
 {bone1Name=dyn_tuft_A, bone2Name=dyn_bearHead, swing1LimitDegrees=20, swing2LimitDegrees=20, twistLimitDegrees=20}
 {bone1Name=dyn_tuft_B, bone2Name=dyn_bearHead, swing1LimitDegrees=20, swing2LimitDegrees=20, twistLimitDegrees=20}
]
